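; BP_S24_PropQuestActor_Ring is a ring prop quest actor for the Season 24 battle pass narrative quests, Part 03.
; Its construction script runs the parent's, then sets the inherited post-interaction ambient particle component's world rotation to match the root component.
; It also sets that effect's FXColorRed parameter to true.
; BeginPlay, Tick and overlap events are unused stubs, so the graph adds no runtime logic of its own.
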(event EventBeginPlay
  (|CallParentFunction) ; This node is disabled and will not be called.\nDrag off pins to build functionality.
)

(event Collision|EventActorBeginOverlap
  (params (OtherActor :type Actor%20Object%20Reference))
  (|CallParentFunction OtherActor) ; This node is disabled and will not be called.\nDrag off pins to build functionality.
)

(event EventTick
  (params (DeltaSeconds :type Float%20%28single-precision%29))
  (|CallParentFunction DeltaSeconds) ; This node is disabled and will not be called.\nDrag off pins to build functionality.
)

(fn ConstructionScript
  (bind _n_postinteractionambient (Variables|Components|GetN_PostInteractionAmbient))
  (|CallParentFunction)
  (Transformation|SetWorldRotation _n_postinteractionambient (Transformation|GetWorldRotation (Variables|Transformation|GetRootComponent)) false false)
  (Effects|Components|ParticleSystem|SetBoolParameter _n_postinteractionambient "FXColorRed" true))
